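; Deco placement tool for the Missile Battery backpack item, handling its placement preview and temporary hiding.
; BeginPlay caches the instigator pawn; when a preview is set, the tool stores the deco helper and passes that pawn to the Missile Battery preview.
; HideFor hides the weapon mesh and starts a timer that calls Show to restore it.
; Despite its name, HideMeshAndPlacement only toggles mesh visibility, requiring both mesh and helper to be valid.
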(event EventBeginPlay
  (bind _asfort_pawn (Utilities|Casting|CastToFortPawn (Game|GetInstigator)))
  (Variables|Default|SetForPawn _asfort_pawn))

(event DecoTool|EventOnSetPreview
  (params (FortDecoHelper :type Fort%20Deco%20Helper%20Object%20Reference))
  ;[1 1 1] Set helper on preview
  (bind _output_get (Variables|Default|SetHelper FortDecoHelper))
  (Utilities|IsValid _output_get
    (:Is%20Valid
      (bind _asb_placement_preview_pack_missle_battery (Utilities|Casting|CastToB_Placement_Preview_Pack_MissleBattery (Variables|/Script/FortniteGame.FortDecoHelper|GetDecoPreview _output_get)))
      (Class|BPlacementPreviewPackMissleBattery|PawnSetHere _asb_placement_preview_pack_missle_battery (Variables|Default|GetForPawn)))))

(event Custom|Show
  ;[1 1 1] Hides this placement tool for InTime seconds
  (CallFunction|HideMeshandPlacement false)
  (Utilities|Time|ClearandInvalidateTimerbyHandle (Variables|Default|GetHideTimerHandle)))

(event Custom|HideFor
  (params (InTime :type Float%20%28double-precision%29))
  ;[1 1 1] Hides this placement tool for InTime seconds
  (CallFunction|HideMeshandPlacement true)
  (bind _returnvalue (Utilities|Time|SetTimerbyEvent (delegate Custom|Show) InTime))
  (Variables|Default|SetHideTimerHandle _returnvalue))

(fn HideMeshAndPlacement
  (params (bInHide :type Boolean))
  (if (and (Utilities|IsValid (Variables|Visual|GetWeaponMesh) :!class /Script/Engine.KismetSystemLibrary) (Utilities|IsValid (Variables|Default|GetHelper) :!class /Script/Engine.KismetSystemLibrary))
    (Rendering|SetVisibility (Variables|Visual|GetWeaponMesh) (not bInHide) true)))

(fn ConstructionScript
  (|CallParentFunction))
