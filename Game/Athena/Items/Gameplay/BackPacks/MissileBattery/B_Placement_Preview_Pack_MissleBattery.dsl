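; Placement preview actor for the missile battery backpack that draws the predicted missile arc for the locally controlled pawn.
; Setting the pawn spawns a hidden trajectory actor and caches physics data from the missile projectile class.
; A looping timer rebuilds the spline from clamped, inverted aim pitch, using the higher of pitch-lerped speed and projectile max speed.
; EndPlay clears the timer and destroys the trajectory actor.
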(event EventBeginPlay)

(event Collision|EventActorBeginOverlap
  (params (OtherActor :type Actor%20Object%20Reference)))

(event EventTick
  (params (DeltaSeconds :type Float%20%28single-precision%29)))

(event PlacementVisuals|EventOnUpdateVisuals
  (params (DeltaSeconds :type Float%20%28single-precision%29) (bSafeToPlace :type Boolean))
  (if bSafeToPlace
    _))

(event Custom|PawnSetHere
  (params (FortPawn :type Fort%20Pawn%20Object%20Reference))
  (bind _output_get (Variables|Default|SetFortPawn FortPawn))
  (if (Pawn|IsLocallyControlled _output_get)
    (Utilities|FlowControl|Sequence
      (:then_0
        (CallFunction|SetUpDummyProjectile (Variables|Default|GetFortPawn))
        (Rendering|SetActorHiddenInGame (Variables|Default|GetTrajectoryActorReference) true))
      (:then_1
        (Utilities|FlowControl|Delay 0.2)
        (Utilities|Time|SetTimerbyFunctionName 0 "UpdateTrajectorySpline" (Variables|Default|GetTrajectoryUpdateInterval) true)))))

(event EventEndPlay
  (params (EndPlayReason :type EEndPlayReason%20Enum :default "Destroyed"))
  (Utilities|Time|ClearTimerbyFunctionName 0 "UpdateTrajectorySpline")
  (Actor|DestroyActor (Variables|Default|GetTrajectoryActorReference)))

(fn CalcRocketSpeedFromPitch
  (params (AimPitch :type Float%20%28double-precision%29))
  (return (Math|Float|Lerp (Variables|LaunchSpeed|GetRocketSpeedMin) (Variables|LaunchSpeed|GetRocketSpeedMax) (/ (- (Math|Float|Clamp(Float) (Math|Rotator|NormalizeAxis AimPitch) (Variables|LaunchSpeed|GetMinSpeedPitch) (Variables|LaunchSpeed|GetMaxSpeedPitch)) (Variables|LaunchSpeed|GetMinSpeedPitch)) (- (Variables|LaunchSpeed|GetMaxSpeedPitch) (Variables|LaunchSpeed|GetMinSpeedPitch))))))

(fn GetGadgetItem
  (bind _controller (CallFunction|GetPlayerController))
  (Utilities|IsValid _controller
    (:Is%20Valid
      (return (Inventory|GetInventoryItemWithGuid _controller (FortWeapon|GetInventoryGUID (Variables|/Script/FortniteGame.FortPawn|GetCurrentWeapon (Variables|Default|GetFortPawn))))))
    (:Is%20Not%20Valid
      (return 0))))

(fn GetPlayerController
  (Utilities|IsValid (Variables|Default|GetFortPawn)
    (:Is%20Valid
      (bind _asfort_player_controller_zone (Utilities|Casting|CastToFortPlayerControllerZone (Pawn|GetController (Variables|Default|GetFortPawn)))
        (:then
          (return _asfort_player_controller_zone))
        (:CastFailed
          _)))
    (:Is%20Not%20Valid
      _))
  (return 0))

(fn SetUpDummyProjectile
  (params (PlayerPawn :type Fort%20Pawn%20Object%20Reference))
  (bind _returnvalue (Game|SpawnActorfromClass (Variables|Default|GetTrajectoryPreviewActor) (Transformation|GetActorTransform PlayerPawn) "AlwaysSpawn" "SelectDefaultAtRuntime" 0 PlayerPawn))
  (bind _asbp_projectile_trajectory (Utilities|Casting|CastToBP_ProjectileTrajectory _returnvalue))
  (Variables|Default|SetTrajectoryActorReference _asbp_projectile_trajectory)
  (bind _trajectorydata (Class|FortKismetLibrary|GetProjectileTrajectoryData PlayerPawn (Variables|Default|GetPrj_MIssileBattery) 0.0 (Variables|Default|GetGravityScale)))
  (Variables|TrajectoryDisplay|SetDummyMaxSpeed (Utilities|Struct|BreakFortProjectileTrajectoryData _trajectorydata))
  (Variables|TrajectoryDisplay|SetDummyGravity (.GravityZ (Utilities|Struct|BreakFortProjectileTrajectoryData _trajectorydata)))
  (Variables|TrajectoryDisplay|SetDummyExtent (Math|Float|Max(Float) (.CapsuleRadius (Utilities|Struct|BreakFortProjectileTrajectoryData _trajectorydata)) (.CapsuleHalfHeight (Utilities|Struct|BreakFortProjectileTrajectoryData _trajectorydata))))
  (Variables|TrajectoryDisplay|SetDummyShouldBounce (.bShouldBounce (Utilities|Struct|BreakFortProjectileTrajectoryData _trajectorydata)))
  (Variables|TrajectoryDisplay|SetDummyBounciness (.Bounciness (Utilities|Struct|BreakFortProjectileTrajectoryData _trajectorydata)))
  (Variables|TrajectoryDisplay|SetDummyFriction (.Friction (Utilities|Struct|BreakFortProjectileTrajectoryData _trajectorydata))))

(fn SpawnLocationAndRotation
  (return (+ (Transformation|GetActorLocation (Variables|Default|GetFortPawn)) (Math|Vector|RotateVectorAroundAxis (Variables|Default|GetGrenadeTargetingOriginOffset) (.yaw (Transformation|GetActorRotation (Variables|Default|GetFortPawn))) "0, 0,1.000000")) (Math|Rotator|MakeRotator (.roll (Targeting|GetDefaultTargetingRotation (Variables|Default|GetFortPawn))) (Math|Float|Clamp(Float) (+ (* (.pitch (Targeting|GetDefaultTargetingRotation (Variables|Default|GetFortPawn))) -1.0) (Variables|Default|GetAdditionalLaunchAngle)) (Variables|LaunchSpeed|GetMinLaunchPitch) (Variables|LaunchSpeed|GetMaxLaunchPitch)) (.yaw (Targeting|GetDefaultTargetingRotation (Variables|Default|GetFortPawn))))))

(fn UpdateTrajectorySpline
  (bind (_outloc _outrot) (CallFunction|SpawnLocationandRotation))
  (Rendering|SetActorHiddenInGame (Variables|Default|GetTrajectoryActorReference) false)
  (bind (_outsplinepoints _outsplinetangents) (Projectile|CalculateProjectileTrajectorySpline :Instigator (Variables|Default|GetFortPawn) :InitialLocation _outloc :InitialVelocity (* (Math|Rotator|GetRotationXVector _outrot) (Math|Float|Max(Float) (CallFunction|CalcRocketSpeedfromPitch (.pitch _outrot)) (Variables|TrajectoryDisplay|GetDummyMaxSpeed))) :MaxSpeed (Math|Float|Max(Float) (CallFunction|CalcRocketSpeedfromPitch (.pitch _outrot)) (Variables|TrajectoryDisplay|GetDummyMaxSpeed)) :Gravity (Variables|TrajectoryDisplay|GetDummyGravity) :Friction (Variables|TrajectoryDisplay|GetDummyFriction) :Bounciness (Math|Float|SelectFloat (Variables|TrajectoryDisplay|GetDummyBounciness) 0.0 (Variables|TrajectoryDisplay|GetDummyShouldBounce)) :TimeStep (Variables|TrajectoryDisplay|GetDummyTimeStep) :TraceExtent (Variables|TrajectoryDisplay|GetDummyExtent) :TraceChannel "ECC_GameTraceChannel8" :MaxBounces (Variables|Default|GetMaxTrajectoryBounces) :MaxSteps 70 :MaxDistanceBetweenSplinePoints 1.5e+03 :InitialDistance 0.0 :LinearDamping 0.0 :bReturnPhysicalMaterial false))
  (Trajectory|SetTrajectorySpline (Variables|Default|GetTrajectoryActorReference) _outsplinepoints _outsplinetangents))

(fn ConstructionScript)
